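# Material function that masks storm cloud emissive and opacity by safe zone scale, behind the UseStormMasking static switch.
# A scaled world-space sphere mask is maxed with an exponential radial gradient on a raised virtual plane, both radii lerped by SafeZoneScale.
# CloudStartingBias lifts the mask toward one, and DebugAngleMask blends to the secondary gradient alone.
# View-angle and distance branches never reach the output.
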MaterialFunction /Game/Effects/Fort_Effects/FrontEnd/S35/MF_Jethro_StormMask1
graph {nodes [
  Constant3Vector()
  n1 = FunctionInput() {inputType=FunctionInput_MaterialAttributes}
  n2 = GetMaterialAttributes(None=@n1) {attributeGetTypes [#1, #2]}
  n3 = WorldPosition()
  n4 = VectorParameter() {group=Storm masking, parameterName=Jethro.StormMaskScale, defaultValue=<rgba 1 1 1.45 1>}
  n5 = ScalarParameter() {defaultValue=443641.75, group=Storm masking, parameterName=Jethro.StormMaskRadius.Min}
  n6 = ScalarParameter() {defaultValue=834971, group=Storm masking, parameterName=Jethro.StormMaskRadius.Max}
  n7 = Saturate(None=CollectionParameter() {collection=/Game/Athena/SafeZone/MPC_SafeZoneClouds, parameterName=SafeZoneScale})
  n8 = Power(
   Base=SphereMask(
    A=Multiply(A=@n3.XYZ, B=@n4.RGB)
    B=VectorParameter() {group=Storm masking, parameterName=Jethro.StormMaskCenter, defaultValue=<rgba 0 0 0 1>}.RGB
    Radius=LinearInterpolate(A=@n5, B=@n6, Alpha=@n7)
    Hardness=ScalarParameter() {defaultValue=0.85, group=Storm masking, parameterName=Jethro.StormMaskHardness}
   )
   Exp=ScalarParameter() {defaultValue=3, group=Storm masking, parameterName=Jethro.StormMaskContrast}
  )
  n9 = Constant3Vector() {constant=<rgba 0 0 1 0>}
  n10 = Constant3Vector() {constant=<rgba 1 0 0 0>}
  n11 = MaterialFunctionCall(Z=ScalarParameter() {defaultValue=190000, parameterName=SecondaryMask.Height}) {materialFunction=&MakeFloat3}
  n12 = Constant() {r=100}
  n13 = ScalarParameter() {defaultValue=1200, parameterName=SecondaryMaskRadius.End}
  n14 = ScalarParameter() {defaultValue=7600, parameterName=SecondaryMaskRadius.Start}
  n15 = Saturate(None=CollectionParameter() {collection=&MPC_SafeZoneClouds, parameterName=SafeZoneScale})
  n16 = Saturate(None=MaterialFunctionCall(
    UVs=MaterialFunctionCall(Plane Normal Axis=@n9, Plane X Axis=@n10, Plane Center=@n11.Result, UV Scale=@n12) {materialFunction=&VirtualPlaneCoordinates}.UVs
    Radius=LinearInterpolate(A=@n13, B=@n14, Alpha=@n15) {constA=0.4}
    Density=ScalarParameter() {defaultValue=7, parameterName=SecondaryMaskDensity}
   ) {materialFunction=&RadialGradientExponential}.RadialGradientExponential)
  n17 = CollectionParameter() {collection=&MPC_SafeZoneClouds, parameterName=CloudStartingBias}
  n18 = LinearInterpolate(
   A=LinearInterpolate(A=Max(A=@n8, B=@n16), Alpha=Saturate(None=@n17))
   B=@n16
   Alpha=ScalarParameter() {group=Storm masking, parameterName=DebugAngleMask}
  )
  n19 = LinearInterpolate(B=@n2.EmissiveColor, Alpha=@n18)
  n20 = LinearInterpolate(B=@n2.Opacity, Alpha=@n18)
  FunctionOutput(None=StaticSwitchParameter(
    True=SetMaterialAttributes!(MaterialAttributes=@n2.MaterialAttributes, EmissiveColor=@n19, Opacity=@n20) {attributeSetTypes [#1, #2]}
    False=@n1
   ) {defaultValue=true, group=Storm masking, parameterName=UseStormMasking})
  n21 = Saturate(None=CollectionParameter() {collection=&MPC_SafeZoneClouds, parameterName=SafeZoneScale})
  n22 = InverseLinearInterpolate(
   A=ScalarParameter() {defaultValue=500000, group=Storm masking, parameterName=DistanceClose}
   B=ScalarParameter() {defaultValue=2000000, group=Storm masking, parameterName=DistanceFar}
   Value=Length(None=WorldPosition() {worldPositionShaderOffset=WPT_CameraRelative}.XY) {is=class MaterialExpressionLength}
  )
  n23 = Subtract(A=CameraPositionWS(), B=WorldPosition().XYZ)
  n24 = Clamp(None=DotProduct(A=Normalize(VectorInput=@n23), B=Constant3Vector() {constant=<rgba 0 0 -1 0>}))
  n25 = Constant() {r=1}
  n26 = Constant()
  n27 = SmoothStep(
   Min=LinearInterpolate(A=LinearInterpolate(Alpha=@n21) {constA=0.4}, B=LinearInterpolate(Alpha=@n21) {constA=0.7}, Alpha=Saturate(None=@n22)) {constA=0.4, constB=0.7}
   Value=Saturate(None=MaterialFunctionCall(Input=@n24, Target Low=@n25, Target High=@n26) {materialFunction=&RemapValueRange}.Result)
  ) {constMax=0}
  MaterialFunctionCall!(Vector3=@n27, MaximumNumberOfDigits, UVs, DebugTextLocation RG_UpperRight BA_LowerLeft, Component Spacing) {materialFunction=/Engine/Functions/Engine_MaterialFunctions02/Utility/DebugFloat3Values}
  n28 = ComponentMask(None=MaterialFunctionCall() {materialFunction=&CameraDirectionVector}.Result) {b=true}
  Multiply(A=SmoothStep(Value=Saturate(None=@n28)) {constMax=0.1, constMin=0.05}, B=@n16)
  Saturate(None=MaterialFunctionCall(Input=@n27, Input High=Constant() {r=0.7}) {materialFunction=&RemapValueRange}.Result)
  SmoothStep(Value=@n16) {constMax=0.2, constMin=0.05}
 ]}
