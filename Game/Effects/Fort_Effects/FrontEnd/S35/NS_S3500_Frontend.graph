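# Niagara system for the Season 35 frontend: drifting foreground smoke, fog and wisps over a backdrop of flying TV and XV mesh bodies firing at speed.
# TV_Body records position, velocity and color through WriteDeath into system arrays; system death values drive XV_Body spawning and velocity.
# Each body emitter feeds a projectile emitter via particle reads, giving green TV and red XV shots.
# Height-based fake distance and region masks adjust visibility; snow emitters are disabled.
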NiagaraSystem /Game/Effects/Fort_Effects/FrontEnd/S35/NS_S3500_Frontend
dependencies {
 usedDataInterfaces [
  &NiagaraDataInterfaceCurve
  &NiagaraDataInterfacePlatformSet
  &NiagaraDataInterfaceParticleRead
  &NiagaraDataInterfaceSpriteRendererInfo
  &NiagaraDataInterfaceArrayPosition
  &NiagaraDataInterfaceArrayFloat3
  &NiagaraDataInterfaceArrayColor
 ]
 usedDynamicInputs [&Multiply_Vector2D_ByFloat, &MakeIntFromFloat]
 usedModules [
  &self:SystemSpawnScript.SystemScriptSource.SystemScriptGraph.NiagaraNodeAssignment_0.SetVariables_#1
  &self:SystemSpawnScript.SystemScriptSource.SystemScriptGraph.NiagaraNodeAssignment_1.SetVariables_#2
  &self:SystemSpawnScript.SystemScriptSource.SystemScriptGraph.NiagaraNodeAssignment_2.SetVariables_#3
  &SystemState
  &self:ReadDeath
  &self:SystemSpawnScript.SystemScriptSource.SystemScriptGraph.NiagaraNodeAssignment_3.SetVariables_#4
 ]
}
emitters {
 Foreground_FlatSmoke {
  emitterUpdateScript [
   &EmitterState(Life Cycle Mode=::NewEnumerator1, Loop Duration=2, UseLoopDelay=false)
   &SpawnRate(SpawnRate=3, Use Spawn Probability=false)
  ]
  particleSpawnScript [
   &InitializeParticle(
    Color=<rgba 0.4844 0.3619 0.2397 0.05>
    Lifetime Max=10
    Lifetime Min=7
    Lifetime Mode=::NewEnumerator1
    Sprite Rotation Mode=::NewEnumerator2
    Sprite Size Max=<400 700>
    Sprite Size Min=<300 450>
    Sprite Size Mode=::NewEnumerator2
    Sprite UV Mode=::NewEnumerator3
    UsePositionOffset=false
    Write Material Random=true
   )
   &ShapeLocation(
    Box Midpoint=&RandomRangeVector(Minimum=<0 0 0>, Maximum=<1 1 1>)
    Box Size=<500 1000 0>
    MaskSpawnGroup=false
    Offset=<0 -100 -123>
    Offset Mode=::NewEnumerator0
    Override Seed=false
    Shape Primitive=::NewEnumerator2
    Surface Only=false
    UseOwnerScale=true
   )
   &AddVelocity(Velocity=<0 -1 0>, Velocity Speed Scale=&RandomRangeFloat(Minimum=45, Maximum=60))
   &CameraOffset(Camera Offset Amount=150)
   /SetVariables(Particles.SpriteAlignment=<0 0 1>) {enabled=false, isSetParameters=true}
   /SetVariables(Particles.SpriteFacing=<0 0.1 1>) {isSetParameters=true}
  ]
  particleUpdateScript [
   &ParticleState()
   &DynamicMaterialParameters(
    Index 2 Param 1=1
    Index 2 Param 2=&Lerp_Float(A=&RandomRangeFloat(Minimum=0, Maximum=0.05), B=0.2, Alpha=$Particles.NormalizedAge)
    Index 2 Param 3=1
    Index 2 Param 4=1
    Module.Index2Param0Write=true
    Module.Index2Param1Write=true
    Module.Index2Param2Write=true
    Module.Index2Param3Write=true
    Write Parameter Index 0=false
    Write Parameter Index 2=true
   )
   &SpriteRotationRate(Rotation Rate=&RandomRangeFloat(Minimum=6, Maximum=30)) {enabled=false}
   &ScaleSpriteSize(Scale Factor=&Vector2DFromFloat(Value=&FloatFromCurve(FloatCurve=DataInterface() {propertyValues {Curve {keys [{tangentMode=RCTM_User, time=0, value=0.75}, {tangentMode=RCTM_User, time=1, value=1}]}}})))
   &ScaleColor(
    Scale Alpha=&FloatFromCurve(FloatCurve=DataInterface() {propertyValues {Curve {keys [<at 0 0 RCIM_Cubic>,<at 0.5 1 RCIM_Cubic>,<at 1 0 RCIM_Cubic>]}}})
    ScaleA=true
    ScaleRGB=true
   )
   &SolveForcesAndVelocity(Clamp Velocity=false, Limit Acceleration=false)
   &SubUVAnimation(
    Random Start Frame=true
    "Sprite Renderer "=DataInterface() {propertyValues {SpriteRenderer=&self:Foreground_FlatSmoke_0.NiagaraSpriteRendererProperties_1}}
    SubUV Lookup Index:NiagaraFloat=&Multiply_Float(A=$Particles.NormalizedAge, B=1)
    UseEndFrame=false
    UseStartFrame=false
   )
  ]
  rendererClasses [&NiagaraSpriteRendererProperties]
  renderers [
   {rendererClass=&NiagaraSpriteRendererProperties}
   {rendererClass=&NiagaraSpriteRendererProperties, rendererIndex=1}
  ]
 }
 Foreground_Fog {
  bEnabled=false
  emitterUpdateScript [
   &EmitterState(Life Cycle Mode=::NewEnumerator1, Loop Duration=2, UseLoopDelay=false)
   &SpawnBurst_Instantaneous(Use Spawn Probability=false, UseLoopCountLimit=false)
   &SpawnRate(SpawnRate=4, Use Spawn Probability=false)
  ]
  particleSpawnScript [
   &InitializeParticle(
    Color Maximum=<rgba 0.585 0.8697 0.9 0.1>
    Color Minimum=<rgba 0.7105 0.9246 0.9473 0.01>
    Color Mode=::NewEnumerator2
    Lifetime Max=10
    Lifetime Min=7
    Lifetime Mode=::NewEnumerator1
    Sprite Rotation Mode=::NewEnumerator2
    Sprite Size Mode=::NewEnumerator1
    Sprite UV Mode=::NewEnumerator3
    Uniform Sprite Size Max=300
    Uniform Sprite Size Min=200
    UsePositionOffset=false
    Write Material Random=true
   )
   &ShapeLocation(
    Box Midpoint=&RandomRangeVector(Minimum=<0 0 0>, Maximum=<1 1 1>)
    Box Size=<400 1000 50>
    MaskSpawnGroup=false
    Offset=<0 250 -50>
    Offset Coordinate Space=::World
    Offset Mode=::NewEnumerator0
    Override Seed=false
    Shape Primitive=::NewEnumerator2
    Surface Only=false
    UseOwnerScale=true
   )
   &DynamicMaterialParameters(Write Parameter Index 0=false)
   &AddVelocity(
    Velocity=&RandomRangeVector(Minimum=<0 -1 0.5>, Maximum=<0 -3 -2>)
    Velocity Speed Scale=&RandomRangeFloat(Minimum=5, Maximum=30)
   )
   &CameraOffset(Camera Offset Amount=150)
  ]
  particleUpdateScript [
   &ParticleState()
   &SpriteRotationRate(Rotation Rate=&RandomRangeFloat(Minimum=5, Maximum=15))
   &ScaleSpriteSize(Scale Factor=&Vector2DFromFloat(Value=&FloatFromCurve(FloatCurve=DataInterface() {propertyValues {Curve {keys [{tangentMode=RCTM_User, time=0, value=0.75}, {tangentMode=RCTM_User, time=1, value=1}]}}})))
   &ScaleColor(
    Scale Alpha=&FloatFromCurve(FloatCurve=DataInterface() {propertyValues {Curve {keys [<at 0 0 RCIM_Cubic>,<at 0.5 1 RCIM_Cubic>,<at 1 0 RCIM_Cubic>]}}}, Scale Curve=1.25)
    ScaleA=true
    ScaleRGB=true
   )
   &SolveForcesAndVelocity(Clamp Velocity=false, Limit Acceleration=false)
  ]
  rendererClasses [&NiagaraSpriteRendererProperties]
  renderers [
   {rendererClass=&NiagaraSpriteRendererProperties}
   {rendererClass=&NiagaraSpriteRendererProperties, rendererIndex=1}
  ]
 }
 Foreground_Fog001 {
  emitterUpdateScript [
   &EmitterState(Life Cycle Mode=::NewEnumerator1, Loop Duration=2, UseLoopDelay=false)
   &SpawnRate(SpawnRate=5, Use Spawn Probability=false)
  ]
  particleSpawnScript [
   &InitializeParticle(
    Color=<rgba 0.4844 0.3619 0.2397 0.1>
    Lifetime Max=10
    Lifetime Min=7
    Lifetime Mode=::NewEnumerator1
    Sprite Rotation Mode=::NewEnumerator2
    Sprite Size Mode=::NewEnumerator1
    Sprite UV Mode=::NewEnumerator3
    Uniform Sprite Size Max=400
    Uniform Sprite Size Min=200
    UsePositionOffset=false
    Write Material Random=true
   )
   &ShapeLocation(
    Box Midpoint=&RandomRangeVector(Minimum=<0 0 0>, Maximum=<1 1 1>)
    Box Size=<150 1000 70>
    MaskSpawnGroup=false
    Offset=<0 0 -150>
    Offset Mode=::NewEnumerator0
    Override Seed=false
    Shape Primitive=::NewEnumerator2
    Surface Only=false
    UseOwnerScale=true
   )
   &AddVelocity(Velocity=<0 -1 0>, Velocity Speed Scale=&RandomRangeFloat(Minimum=30, Maximum=45))
   &CameraOffset(Camera Offset Amount=150)
  ]
  particleUpdateScript [
   &ParticleState()
   &DynamicMaterialParameters(
    Index 2 Param 1=1
    Index 2 Param 2=&Lerp_Float(A=&RandomRangeFloat(Minimum=0, Maximum=0.05), B=0.2, Alpha=$Particles.NormalizedAge)
    Index 2 Param 3=1
    Index 2 Param 4=1
    Module.Index2Param0Write=true
    Module.Index2Param1Write=true
    Module.Index2Param2Write=true
    Module.Index2Param3Write=true
    Write Parameter Index 0=false
    Write Parameter Index 2=true
   )
   &SpriteRotationRate(Rotation Rate=&RandomRangeFloat(Minimum=6, Maximum=30))
   &ScaleSpriteSize(Scale Factor=&Vector2DFromFloat(Value=&FloatFromCurve(FloatCurve=DataInterface() {propertyValues {Curve {keys [{tangentMode=RCTM_User, time=0, value=0.75}, {tangentMode=RCTM_User, time=1, value=1}]}}})))
   &ScaleColor(
    Scale Alpha=&FloatFromCurve(FloatCurve=DataInterface() {propertyValues {Curve {keys [<at 0 0 RCIM_Cubic>,<at 0.5 1 RCIM_Cubic>,<at 1 0 RCIM_Cubic>]}}})
    ScaleA=true
    ScaleRGB=true
   )
   &SolveForcesAndVelocity(Clamp Velocity=false, Limit Acceleration=false)
  ]
  rendererClasses [&NiagaraSpriteRendererProperties]
  renderers [{rendererClass=&NiagaraSpriteRendererProperties}]
 }
 Foreground_Wisps {
  emitterUpdateScript [
   &EmitterState(Life Cycle Mode=::NewEnumerator1, Loop Duration=2, UseLoopDelay=false)
   &SpawnRate(SpawnRate=2, Use Spawn Probability=false)
  ]
  particleSpawnScript [
   &InitializeParticle(
    Color=<rgba 0.4844 0.3619 0.2397 0.05>
    Lifetime Max=10
    Lifetime Min=7
    Lifetime Mode=::NewEnumerator1
    Sprite Size Max=<200 550>
    Sprite Size Min=<150 200>
    Sprite Size Mode=::NewEnumerator2
    Sprite UV Mode=::NewEnumerator3
    UsePositionOffset=false
    Write Material Random=true
   )
   &ShapeLocation(
    Box Midpoint=&RandomRangeVector(Minimum=<0 0 0>, Maximum=<1 1 1>)
    Box Size=<150 1000 70>
    MaskSpawnGroup=false
    Offset=<0 0 -150>
    Offset Mode=::NewEnumerator0
    Override Seed=false
    Shape Primitive=::NewEnumerator2
    Surface Only=false
    UseOwnerScale=true
   )
   &AddVelocity(Velocity=<0 -1 0>, Velocity Speed Scale=&RandomRangeFloat(Minimum=30, Maximum=45))
   &CameraOffset(Camera Offset Amount=150)
  ]
  particleUpdateScript [
   &ParticleState()
   &DynamicMaterialParameters(
    Index 2 Param 1=1
    Index 2 Param 2=&Lerp_Float(A=&RandomRangeFloat(Minimum=0, Maximum=0.05), B=0.2, Alpha=$Particles.NormalizedAge)
    Index 2 Param 3=1
    Index 2 Param 4=1
    Module.Index2Param0Write=true
    Module.Index2Param1Write=true
    Module.Index2Param2Write=true
    Module.Index2Param3Write=true
    Write Parameter Index 0=false
    Write Parameter Index 2=true
   )
   &SpriteRotationRate(Rotation Rate=&RandomRangeFloat(Minimum=6, Maximum=30)) {enabled=false}
   &ScaleSpriteSize(Scale Factor=&Vector2DFromFloat(Value=&FloatFromCurve(FloatCurve=DataInterface() {propertyValues {Curve {keys [{tangentMode=RCTM_User, time=0, value=0.75}, {tangentMode=RCTM_User, time=1, value=1}]}}})))
   &ScaleColor(
    Scale Alpha=&FloatFromCurve(FloatCurve=DataInterface() {propertyValues {Curve {keys [<at 0 0 RCIM_Cubic>,<at 0.5 1 RCIM_Cubic>,<at 1 0 RCIM_Cubic>]}}})
    ScaleA=true
    ScaleRGB=true
   )
   &SolveForcesAndVelocity(Clamp Velocity=false, Limit Acceleration=false)
   &SubUVAnimation(
    Random Start Frame=true
    "Sprite Renderer "=DataInterface() {propertyValues {SpriteRenderer=&self:Foreground_Wisps_0.NiagaraSpriteRendererProperties_1}}
    SubUV Lookup Index:NiagaraFloat=&Multiply_Float(A=$Particles.NormalizedAge, B=1.3)
    UseEndFrame=false
    UseStartFrame=false
   )
  ]
  rendererClasses [&NiagaraSpriteRendererProperties]
  renderers [{rendererClass=&NiagaraSpriteRendererProperties}]
 }
 SnowflakesAir {
  bEnabled=false
  simTarget=GPUComputeSim
  emitterSpawnScript [/SetVariables(
    Emitter.bool:NiagaraBool=&IsPlatformSetActive(New Platform Set=DataInterface() {propertyValues {Platforms {qualityLevelMask=-4}}})
    Emitter.float:NiagaraFloat=&MakeCustomFloatFromBool(
     Bool=&IsPlatformSetActive(New Platform Set=DataInterface() {propertyValues {Platforms {qualityLevelMask=-29}}})
     True Float=1
     False Float=&MakeCustomFloatFromBool(
      Bool=&IsPlatformSetActive(New Platform Set=DataInterface() {propertyValues {Platforms {qualityLevelMask=-28}}})
      True Float=1
      False Float=0
     )
    )
    Emitter.float001:NiagaraFloat=&MakeCustomFloatFromBool(
     Bool=&IsPlatformSetActive(New Platform Set=DataInterface() {propertyValues {Platforms {qualityLevelMask=-29}}})
     True Float=1
     False Float=0
    )
   ) {isSetParameters=true}]
  emitterUpdateScript [
   &EmitterState()
   &SpawnRate(SpawnRate=750, Use Spawn Probability=false)
  ]
  particleSpawnScript [
   /Niagara/Modules/Spawn/Initialization/InitializeParticle(
    Lifetime=&UniformRangedFloat(Minimum=3, Maximum=6)
    Module.Write Color=true
    Module.Write Lifetime=true
    Module.Write Mass=true
    Module.Write Material Random=true
    Module.Write Position=false
    Module.Write Scale=false
    Module.Write SpriteRotation=true
    Module.Write SpriteSize=true
    Sprite Rotation=&UniformRangedFloat(Minimum=0, Maximum=360)
    Sprite Size=&Vector2DFromFloat(Value=&RandomRangeFloat(Minimum=2, Maximum=10))
   ) {enabled=false}
   /Niagara/Modules/Spawn/Initialization/V2/InitializeParticle(
    Color=<rgba 2.5 2.5 2.5 0.5>
    Lifetime=&UniformRangedFloat(Minimum=3, Maximum=6)
    Mass Mode=::NewEnumerator1
    Position Offset=<200 0 0>
    Position Offset Coordinate Space=::World
    Sprite Rotation Mode=::NewEnumerator2
    Sprite Size=&Vector2DFromFloat(Value=&RandomRangeFloat(Minimum=2, Maximum=10))
    Sprite Size Mode=::NewEnumerator4
    UsePositionOffset=true
    Write Material Random=true
   ) {module=InitializeParticle001}
   /Niagara/Modules/Spawn/Location/SphereLocation(
    Hemisphere Z=true
    MaskSpawnGroup=false
    Module.Override Randomness=false
    Non Uniform Scale=<1 0.5 1>
    Offset=<-400 0 100>
    Sphere Radius=800
    Surface Only=false
    Surface Only Band Thickness=1
   ) {enabled=false}
   &ShapeLocation(MaskSpawnGroup=false, Offset=<-400 0 300>, Offset Mode=::NewEnumerator0, Override Seed=false, Sphere Radius=800, UseOwnerScale=true)
   &AddVelocity(Coordinate Space=::World, Velocity=&UniformRangedVector(Minimum=<100 0 -35>, Maximum=<150 0 -75>))
   &ApplyInitialForces()
  ]
  particleUpdateScript [
   &Collision(
    Advanced Aging Rate=0
    Age Colliding Particles=true
    Analytical Collision Plane Position 1:Vector3f=<0 0 0>
    Analytical Collision Plane Position 2:Vector3f=<0 0 0>
    Friction=100000
    Friction During a Bounce=1000000
    GPU Collision Type=::NewEnumerator3
    KillOnCollision=false
    Module.Control Rest State Via Surface Normal=true
    Module.EnableMaxCollisionCount=false
    Module.Randomize Collision Normal Vector=false
    Percentage of Penetration Before Rest=0.001
    Restitution=0
    Simple Friction=false
    Static Friction=100000
    Static Friction Engagement Speed=100000
   )
   &ParticleState(DeltaTime:NiagaraFloat=&Multiply_Float(
     A=$Engine.DeltaTime
     B=&MakeCustomFloatFromBool(
      Bool=&SetBoolByIntComparison(A=$Particles.Number Of Collisions, B=0, Comparison Type=::NewEnumerator0)
      True Float=3
      False Float=1
     )
    ))
   &KillParticles(Kill Particles=&Boolean_AndOperation(
     Bool1=&SetBoolByIntComparison(A=$Particles.Number Of Collisions, B=0, Comparison Type=::NewEnumerator0)
     Bool2=&SetBoolByFloatComparison(A=$Particles.Age, B=0.1, Comparison Type=::NewEnumerator6)
    ))
   &KillParticles(Kill Particles=&Boolean_AndOperation(
     Bool1=&SetBoolByFloatComparison(A=$Particles.Age, B=0.1, Comparison Type=::NewEnumerator6)
     Bool2=&Boolean_AndOperation(
      Bool1=&SetBoolByIntComparison(A=$Particles.Number Of Collisions, B=0, Comparison Type=::NewEnumerator0)
      Bool2=&SetBoolByFloatComparison(
       A=&MakeFloatFromVector(Vector=Linked() {linkedVariable {name=Particles.Position, type=NiagaraPosition}}, Channel=::NewEnumerator2)
       B=10
       Comparison Type=::NewEnumerator6
      )
     )
    )) {module=KillParticles001}
   &ScaleColor(
    Scale Alpha=&FloatFromCurve(FloatCurve=DataInterface() {propertyValues {Curve {keys [
         <at -0.0007 -0.0138 RCIM_Cubic>
         {arriveTangentWeight=0.2085, interpMode=RCIM_Cubic, leaveTangentWeight=0.1761, tangentMode=RCTM_User, time=0.5, value=1}
         <at 1.0011 0.002 RCIM_Cubic>
        ]}}})
    Scale RGB=&VectorFromFloat(Value=1)
    ScaleA=true
    ScaleRGB=true
   )
   &CurlNoiseForce(
    Noise Frequency=250
    Noise Strength=&MakeCustomFloatFromBool(
     Bool=&SetBoolByIntComparison(A=$Particles.Number Of Collisions, B=0, Comparison Type=::NewEnumerator0)
     True Float=0
     False Float=500
    )
    Pan Noise=false
   ) {module=CurlNoiseForce001}
   &VortexForce(
    UseOriginOffset=false
    Vortex Axis=<1 0 0>
    Vortex Force Amount=&MakeCustomFloatFromBool(
     Bool=&SetBoolByIntComparison(A=$Particles.Number Of Collisions, B=0, Comparison Type=::NewEnumerator0)
     True Float=0
     False Float=20
    )
   ) {enabled=false}
   &GravityForce(Gravity=&MakeCustomVectorFromBool(
     Bool=&SetBoolByIntComparison(A=$Particles.Number Of Collisions, B=0, Comparison Type=::NewEnumerator0)
     True Vector=<0 0 0>
     False Vector=<0 0 -50>
    ))
   &WindForce(
    Wind Speed=<250 0 100>
    Wind Speed Scale=&MakeCustomFloatFromBool(
     Bool=&SetBoolByIntComparison(A=$Particles.Number Of Collisions, B=0, Comparison Type=::NewEnumerator0)
     True Float=0
     False Float=&Lerp_Float(
      A=0.35
      B=0.1
      Alpha=&Waveform("Waveform [1]"=::NewEnumerator9, "[1] Frequency"=0.5, "[1] Amplitude Min/Max"=<0 1>)
     )
    )
   )
   &SolveForcesAndVelocity(Clamp Velocity=true, Limit Acceleration=false, Speed Limit=150)
   &CalculateAccurateVelocity()
   &KillParticles(Kill Particles=&Boolean_AndOperation(
     Bool1=&SetBoolByFloatComparison(A=$Particles.Age, B=0.1, Comparison Type=::NewEnumerator6)
     Bool2=&SetBoolByFloatComparison(A=&VectorLength(Vector=$Particles.Velocity), B=10, Comparison Type=::NewEnumerator6)
    )) {module=KillParticles002}
   &SpriteRotationRate(Rotation Rate=&UniformRangedFloat(Minimum=-50, Maximum=150))
   /SetVariables(Particles.RandomGlintVector:Vector3f=&RotateVectorByQuaternion(
     Vector To Rotate=&RandomVector(Vector Scale=1)
     Quaternion=&MakeQuatFromAxisAngle(
      Axis=&RandomVector(Vector Scale=1)
      AngleInDegrees=&Multiply_Float(A=$Particles.NormalizedAge, B=&RandomRangeFloat(Minimum=2000, Maximum=3000))
     )
    )) {isSetParameters=true, module=SetVariables_#5}
   &DynamicMaterialParameters(
    Index 0 Param 1=&RandomRangeFloat(Minimum=0, Maximum=1)
    Index 0 Param 2=&MakeFloatFromVector(Vector=$Particles.RandomGlintVector, Channel=::NewEnumerator0)
    Index 0 Param 3=&MakeFloatFromVector(Vector=$Particles.RandomGlintVector, Channel=::NewEnumerator1)
    Index 0 Param 4=&MakeFloatFromVector(Vector=$Particles.RandomGlintVector, Channel=::NewEnumerator2)
    Module.Param0WriteEnabled=true
    Module.Param1WriteEnabled=true
    Module.Param2WriteEnabled=true
    Module.Param3WriteEnabled=true
   )
  ]
  rendererClasses [&NiagaraSpriteRendererProperties]
  renderers [{rendererClass=&NiagaraSpriteRendererProperties}]
 }
 SnowflakesAir_LOW {
  bEnabled=false
  emitterUpdateScript [
   &EmitterState()
   &SpawnRate(SpawnRate=50, Use Spawn Probability=false)
  ]
  particleSpawnScript [
   /Niagara/Modules/Spawn/Initialization/InitializeParticle(
    Lifetime=&UniformRangedFloat(Minimum=3, Maximum=6)
    Module.Write Color=true
    Module.Write Lifetime=true
    Module.Write Mass=true
    Module.Write Material Random=true
    Module.Write Position=false
    Module.Write Scale=false
    Module.Write SpriteRotation=true
    Module.Write SpriteSize=true
    Sprite Rotation=&UniformRangedFloat(Minimum=0, Maximum=360)
    Sprite Size=&Vector2DFromFloat(Value=&RandomRangeFloat(Minimum=10, Maximum=15))
   ) {enabled=false}
   /Niagara/Modules/Spawn/Initialization/V2/InitializeParticle(
    Color=<rgba 1 1 1 0.5>
    Lifetime=&UniformRangedFloat(Minimum=3, Maximum=6)
    Mass Mode=::NewEnumerator1
    Sprite Rotation Mode=::NewEnumerator2
    Sprite Size=&Vector2DFromFloat(Value=&RandomRangeFloat(Minimum=10, Maximum=15))
    Sprite Size Mode=::NewEnumerator4
    UsePositionOffset=false
    Write Material Random=true
   ) {module=InitializeParticle001}
   &ShapeLocation(MaskSpawnGroup=false, Offset=<-400 0 300>, Offset Mode=::NewEnumerator0, Override Seed=false, Sphere Radius=800, UseOwnerScale=true)
   &AddVelocity(Velocity=&UniformRangedVector(Minimum=<100 0 -35>, Maximum=<150 0 -75>))
   &ApplyInitialForces()
  ]
  particleUpdateScript [
   &ParticleState() {module=UpdateAge}
   &ScaleColor(
    Scale Alpha=&FloatFromCurve(FloatCurve=DataInterface() {propertyValues {Curve {keys [
         <at -0.0007 -0.0138 RCIM_Cubic>
         {arriveTangent=2.6721, interpMode=RCIM_Cubic, leaveTangent=2.6721, tangentMode=RCTM_User, time=0.3127, value=0.3952}
         {arriveTangent=-0.5712, interpMode=RCIM_Cubic, leaveTangent=-0.5712, tangentMode=RCTM_User, time=0.5424, value=1.0086}
         <at 1.0011 0.002 RCIM_Cubic>
        ]}}})
    ScaleA=true
    ScaleRGB=true
   )
   &CurlNoiseForce(Noise Frequency=250, Noise Strength=500, Pan Noise=false) {module=CurlNoiseForce001}
   &GravityForce(Gravity=<0 0 -50>)
   &WindForce(
    Wind Speed=<400 0 100>
    Wind Speed Scale=&Lerp_Float(
     A=0.35
     B=0.1
     Alpha=&Waveform("Waveform [1]"=::NewEnumerator9, "[1] Frequency"=0.5, "[1] Amplitude Min/Max"=<0 1>)
    )
   )
   &SolveForcesAndVelocity(Clamp Velocity=true, Limit Acceleration=false, Speed Limit=150)
   &SpriteRotationRate(Rotation Rate=&UniformRangedFloat(Minimum=-50, Maximum=150))
   /SetVariables(Particles.RandomGlintVector:Vector3f=&RotateVectorByQuaternion(
     Vector To Rotate=&RandomVector(Vector Scale=1)
     Quaternion=&MakeQuatFromAxisAngle(
      Axis=&RandomVector(Vector Scale=1)
      AngleInDegrees=&Multiply_Float(A=$Particles.NormalizedAge, B=&RandomRangeFloat(Minimum=2000, Maximum=3000))
     )
    )) {isSetParameters=true, module=SetVariables_#5}
   &DynamicMaterialParameters(
    Index 0 Param 1=&RandomRangeFloat(Minimum=0, Maximum=1)
    Index 0 Param 2=&MakeFloatFromVector(Vector=$Particles.RandomGlintVector, Channel=::NewEnumerator0)
    Index 0 Param 3=&MakeFloatFromVector(Vector=$Particles.RandomGlintVector, Channel=::NewEnumerator1)
    Index 0 Param 4=&MakeFloatFromVector(Vector=$Particles.RandomGlintVector, Channel=::NewEnumerator2)
    Module.Param0WriteEnabled=true
    Module.Param1WriteEnabled=true
    Module.Param2WriteEnabled=true
    Module.Param3WriteEnabled=true
   )
  ]
  rendererClasses [&NiagaraSpriteRendererProperties]
  renderers [{rendererClass=&NiagaraSpriteRendererProperties}]
 }
 TV_Body {
  emitterUpdateScript [
   &EmitterState()
   &SpawnRate(
    SpawnRate=&FloatFromCurve(
     FloatCurve=DataInterface() {propertyValues {Curve {defaultValue=1, keys [<at 0 1 RCIM_Constant>,<at 0.3 0 RCIM_Constant>,<at 1 0 RCIM_Constant>]}}}
     CurveIndex=$Emitter.NormalizedLoopAge
     Scale Curve=1.5
    )
    Use Spawn Probability=false
   )
  ]
  particleSpawnScript [
   &FN_SetSpawnGroup(Number of Spawn Groups=2) {enabled=false}
   &PartitionParticles(Partition Mode:ENiagara_PartitionMode_v2=::NewEnumerator5)
   &PartitionParticles() {module=PartitionParticles001}
   &InitializeParticle(
    Lifetime=8
    Mesh Scale Mode=::NewEnumerator3
    Mesh Uniform Scale=0.008
    Position Offset=&MaskVectorBySpawnGroup(
     Masked Vector=&RandomRangeVector(Minimum=<0 800 200>, Maximum=<0 800 350>)
     Passthrough Vector=&RandomRangeVector(Minimum=<0 -800 200>, Maximum=<0 -800 350>)
     Spawn Group Mask=$Particles.PartitionParticles001.Partition
    )
    Sprite Size Mode=::NewEnumerator3
    Uniform Sprite Size=15
    UsePositionOffset=true
    Write Material Random=true
   )
   /SetVariables(Particles.FakeDistance:NiagaraFloat=&RandomRangeFloat(Minimum=0, Maximum=1)) {isSetParameters=true}
   &InitialMeshOrientation(Mesh Orientation Mode=::NewEnumerator2, Use Rotation Vector=false)
   &SpriteFacingAndAlignment(Facing Coordinate Space=::Local, WriteAlignment=true, WriteFacing=true) {enabled=false}
   &AddVelocity(
    Velocity=&NormalizeVector(
     Vector To Normalize=&RandomRangeVector(Minimum=<0 -1 -0.2>, Maximum=<0 -1 0.2>)
     Safe Normalize=true
     Safe Fallback Threshold=0
     Safe Fallback Vector=<0 0 0.001>
    )
    Velocity Speed Scale=&Multiply_Float(
     A=&Add_Float(A=200, B=&Multiply_Float(A=150, B=$Particles.FakeDistance))
     B=&MaskFloatBySpawnGroup(Masked Float=1, Passthrough Float=-1, Spawn Group Mask=$Particles.PartitionParticles001.Partition)
    )
   )
   /SetVariables(Particles.VisibilityTag=$Particles.PartitionParticles.Partition) {enabled=false, isSetParameters=true, module=SetVariables_#6}
  ]
  particleUpdateScript [
   &ParticleState()
   &self:WriteDeath(
    Color Array:NiagaraDataInterfaceArrayColor=$System.Color Array
    Position Array:NiagaraDataInterfaceArrayPosition=$System.DeathArray
    Velocity Array:NiagaraDataInterfaceArrayFloat3=$System.VelocityArray
   ) {module=WriteDeath}
   /Niagara/Modules/Update/Forces/LineAttractionForce(Attraction Strength=-20, Line End:Vector3f=<0 -600 75>, Line Start:Vector3f=<0 600 75>, Use Attraction Falloff=false)
   &RegionMask(
    Invert Region=true
    Region Falloff=0.1
    Region Origin:Vector3f=&AddVectorToPosition(Position=$Engine.Emitter.SimulationPosition, Vector=<0 840 -100>)
    Region Radius=730
    UseOffset=false
    UseRotation=false
   ) {module=RegionMask002}
   &RegionMask(
    Invert Region=true
    Region Dimensions:Vector3f=<100 1200 1000>
    Region Falloff=0.1
    Region Origin:Vector3f=&AddVectorToPosition(Position=$Engine.Emitter.SimulationPosition, Vector=<0 -400 -100>)
    Region Type:ENiagaraDebugDrawShape=::NewEnumerator1
    Rotation:Vector3f=<0 5 0>
    Rotation Angle Type:ENiagara_AngleInput=::NewEnumerator0
    UseOffset=false
    UseRotation=true
   ) {module=RegionMask003}
   /SetVariables(Particles.CombinedMask:NiagaraFloat=&Multiply_Float(A=$Particles.RegionMask003.Float, B=$Particles.RegionMask002.Float)) {isSetParameters=true}
   /SetVariables(Particles.FakeHorizonDistance:NiagaraFloat=&RemapRange(
     Input Value To Remap=&MakeFloatFromVector(
      Vector=&ConvertPositionToVector(Input Position=$Particles.Position, Conversion Operation:ENiagara_LWCConvertPosToVec=::NewEnumerator0)
      Channel=::NewEnumerator2
     )
     Input Min=100
     Input Max=350
     Output Min=0
     Output Max=1
    )) {isSetParameters=true}
   &Drag(
    Drag=&RemapRange(Input Value To Remap=$Particles.FakeHorizonDistance, Input Min=0, Input Max=1, Output Min=0.25, Output Max=0)
    Use Linear Drag=true
    Use Rotational Drag=true
   )
   &ScaleMeshSize(Scale Factor=&Multiply_VectorByFloat(
     Vector=&Add_Vector(A=<1 1 1>, B=&Multiply_VectorByFloat(Vector=<1 1 1>, Float=$Particles.FakeHorizonDistance))
     Float=&Multiply_Float(
      A=$Particles.CombinedMask
      B=&FloatFromCurve(FloatCurve=DataInterface() {propertyValues {Curve {
          defaultValue=1
          keys [{interpMode=RCIM_Cubic, tangentMode=RCTM_User, time=0.7, value=1}, {arriveTangent=-5.7587, tangentMode=RCTM_User, time=1, value=0}]
         }}})
     )
    )) {module=ScaleMeshSize001}
   &ScaleSpriteSize(
    Scale Sprite Size Mode=::NewEnumerator0
    Uniform Scale Factor=&Multiply_Float(A=$Particles.FakeHorizonDistance, B=$Particles.CombinedMask)
   ) {enabled=false}
   &SolveForcesAndVelocity(Clamp Velocity=false, Limit Acceleration=false)
   &FN_Sprite_Attributes(
    Color=<rgba 1 0.03 0 1>
    Create Color Attribute=true
    Create Position Attribute=true
    Position:NiagaraPosition=&LerpPosition(
     A=$Particles.Initial.Position
     B=$Particles.Position
     Alpha=&FloatFromCurve(FloatCurve=DataInterface() {propertyValues {Curve {defaultValue=1, keys [<at 0 0.2>,<at 0.5 0.9>]}}})
    )
    Write to Module Namespace=true
   ) {enabled=false}
   &KillParticlesInVolume(Box Size:Vector3f=<100 1800 1000>, Invert Volume=true, Kill Shape=::NewEnumerator1)
   &DynamicMaterialParameters(
    Index 0 Param 1=100
    Module.Index0Float0Write=true
    Module.Index0Float1Write=true
    Module.Index0Float2Write=true
    Module.Index0Float3Write=true
   )
   &ScaleColor(
    Scale RGB=&MakeVectorFromLinearColorRGB(Color=&Multiply_LinearColorByFloat(
      Linear Color=<rgba 0.2464 0.2759 0.3 1>
      Float=&RemapRange(Input Value To Remap=$Particles.FakeHorizonDistance, Input Min=0, Input Max=1, Output Min=0.9, Output Max=0.15)
     ))
    ScaleA=true
    ScaleRGB=true
   )
  ]
  rendererClasses [&NiagaraSpriteRendererProperties, &NiagaraMeshRendererProperties]
  renderers [
   {rendererClass=&NiagaraSpriteRendererProperties}
   {rendererClass=&NiagaraMeshRendererProperties, rendererIndex=1}
  ]
 }
 TV_Projectile {
  emitterUpdateScript [
   &EmitterState(Life Cycle Mode=::NewEnumerator1, Loop Duration=1, UseLoopDelay=false)
   &SpawnRate(SpawnRate=4, Use Spawn Probability=false) {enabled=false}
   &SpawnParticlesFromOtherEmitter(
    Attribute Reader=DataInterface() {propertyValues {EmitterBinding {bindingMode=Other, emitterName=TV_Body}}}
    Spawn Rate=1
    Spawn Rate Per Particle Cap=1
    Use Maximum Age=false
    Use Minimum Age=false
   )
  ]
  particleSpawnScript [
   &InitializeParticle(
    Color=<rgba 0.0218 2 0 1>
    Lifetime=&RandomRangeFloat(Minimum=1, Maximum=2)
    Mesh Scale Mode=::NewEnumerator3
    Mesh Uniform Scale=0.03
    UsePositionOffset=true
    Write Material Random=true
   ) {module=InitializeParticle001}
   &SampleParticlesFromOtherEmitter(
    Age Sampling:ENiagara_DirectReadSamplingMode=::NewEnumerator1
    Mesh Scale Sampling:ENiagara_DirectReadSamplingMode=::NewEnumerator1
    Velocity Sampling=::NewEnumerator1
   )
   &AddVelocity(
    Rotation Coordinate Space=::World
    Velocity=&NormalizeVector(
     Vector To Normalize=$Output.SampleParticlesFromOtherEmitter.SampledVelocity
     Safe Normalize=true
     Safe Fallback Threshold=0
     Safe Fallback Vector=<0 0 0.001>
    )
    Velocity Speed Scale=&Multiply_Float(A=700, B=1)
   )
   /SetVariables(Particles.Color=&MakeLinearColorFromVectorAndFloat(
     "Vector (RGB)"=&MakeVectorFromLinearColorRGB(Color=<rgba 0.0654 6 0 1>)
     "Float (Alpha)"=&FloatFromCurve(
      FloatCurve=DataInterface() {propertyValues {Curve {
         defaultValue=1
         keys [
          {arriveTangent=9.0725, interpMode=RCIM_Constant, time=0, value=0}
          <at 3 0.9847 RCIM_Constant>
          {arriveTangent=-0.0405, interpMode=RCIM_Constant, leaveTangent=-0.0405, time=8, value=0}
         ]
        }}}
      CurveIndex=$Output.SampleParticlesFromOtherEmitter.SampledAge
     )
    )) {enabled=false, isSetParameters=true}
   /SetVariables(Particles.Scale=&Add_Vector(
     A=&Multiply_VectorByFloat(
      Vector=&VectorFromFloat(Value=&MakeFloatFromVector(Vector=$Output.SampleParticlesFromOtherEmitter.SampledScale, Channel=::NewEnumerator0))
      Float=2
     )
     B=$Particles.Scale
    )) {isSetParameters=true}
  ]
  particleUpdateScript [
   &ParticleState()
   &SolveForcesAndVelocity(Clamp Velocity=false, Limit Acceleration=false)
   &RegionMask(
    Region Origin:Vector3f=&AddVectorToPosition(Position=$Engine.Emitter.SimulationPosition, Vector=<-660 0 -25>)
    Region Radius=500
    UseOffset=false
    UseRotation=false
   ) {module=RegionMask002}
   &RegionMask(
    Region Origin:Vector3f=&AddVectorToPosition(Position=$Engine.Emitter.SimulationPosition, Vector=<700 0 -400>)
    Region Radius=600
    UseOffset=false
    UseRotation=false
   ) {module=RegionMask003}
   &RegionMask(
    Region Origin:Vector3f=&AddVectorToPosition(Position=$Engine.Emitter.SimulationPosition, Vector=<250 0 -150>)
    Region Radius=300
    UseOffset=false
    UseRotation=false
   ) {module=RegionMask004}
   /SetVariables(Particles.CombinedMask:NiagaraFloat=&MakeCustomFloatFromBool(
     Bool=&Boolean_OrOperation(
      Bool1=&Boolean_OrOperation(Bool1=$Particles.RegionMask002.IsInsideMask, Bool2=$Particles.RegionMask003.IsInsideMask)
      Bool2=$Particles.RegionMask004.IsInsideMask
     )
     True Float=0
     False Float=1
    )) {isSetParameters=true}
   &ScaleMeshSize(Scale Factor=&Multiply_VectorByFloat(
     Vector=&MakeVector(
      X=&FloatFromCurve(
       FloatCurve=DataInterface() {propertyValues {Curve {
          defaultValue=1
          keys [
           {arriveTangent=4.1785, interpMode=RCIM_Cubic, leaveTangent=4.1785, tangentMode=RCTM_User, time=0, value=0}
           {arriveTangent=1.1179, interpMode=RCIM_Cubic, leaveTangent=1.1179, tangentMode=RCTM_User, time=0.2042, value=0.6228}
           {arriveTangent=0.0224, interpMode=RCIM_Cubic, leaveTangent=0.0224, tangentMode=RCTM_User, time=1, value=1}
          ]
         }}}
       Scale Curve=&RandomRangeFloat(Minimum=4, Maximum=6)
      )
      Y=2
      Z=2
     )
     Float=$Particles.CombinedMask
    ))
   /SetVariables(Particles.FakeHorizonDistance:NiagaraFloat=&RemapRange(
     Input Value To Remap=&MakeFloatFromVector(
      Vector=&ConvertPositionToVector(Input Position=$Particles.Position, Conversion Operation:ENiagara_LWCConvertPosToVec=::NewEnumerator0)
      Channel=::NewEnumerator2
     )
     Input Min=100
     Input Max=350
     Output Min=0
     Output Max=0.5
    )) {isSetParameters=true}
   &ScaleColor(Scale Alpha=&OneMinusFloat(Float=$Particles.FakeHorizonDistance), ScaleA=true, ScaleRGB=false)
   &KillParticlesInVolume(Box Size:Vector3f=<100 1800 1000>, Invert Volume=true, Kill Shape=::NewEnumerator1)
  ]
  rendererClasses [&NiagaraMeshRendererProperties]
  renderers [{rendererClass=&NiagaraMeshRendererProperties, rendererIndex=0}]
 }
 XV_Body {
  emitterUpdateScript [
   &EmitterState()
   &SpawnPerFrame(Spawn:NiagaraBool=$System.DeathSpawn, Spawn Probability=0.7, Use Spawn Probability=false)
   &SpawnRate(SpawnRate=1.2, Use Spawn Probability=false) {enabled=false}
  ]
  particleSpawnScript [
   &FN_SetSpawnGroup(Number of Spawn Groups=2) {enabled=false}
   &PartitionParticles(Partition Mode:ENiagara_PartitionMode_v2=::NewEnumerator5)
   &PartitionParticles() {module=PartitionParticles001}
   &InitializeParticle(
    Lifetime=8
    Mesh Scale Mode=::NewEnumerator3
    Mesh Uniform Scale=0.008
    Position Offset=&MaskVectorBySpawnGroup(
     Masked Vector=&RandomRangeVector(Minimum=<0 800 200>, Maximum=<0 800 350>)
     Passthrough Vector=&RandomRangeVector(Minimum=<0 -800 200>, Maximum=<0 -800 350>)
     Spawn Group Mask=$Particles.PartitionParticles001.Partition
    )
    Sprite Size Mode=::NewEnumerator3
    Uniform Sprite Size=15
    UsePositionOffset=true
    Write Material Random=true
   )
   /SetVariables(Particles.FakeDistance:NiagaraFloat=&RandomRangeFloat(Minimum=0, Maximum=1)) {isSetParameters=true}
   &InitialMeshOrientation(Mesh Orientation Mode=::NewEnumerator2, Use Rotation Vector=false)
   &SpriteFacingAndAlignment(Facing Coordinate Space=::Local, WriteAlignment=true, WriteFacing=true) {enabled=false}
   &AddVelocity(
    Velocity=&NormalizeVector(
     Vector To Normalize=&RandomRangeVector(Minimum=<0 -1 -0.2>, Maximum=<0 -1 0.2>)
     Safe Normalize=true
     Safe Fallback Threshold=0
     Safe Fallback Vector=<0 0 0.001>
    )
    Velocity Speed Scale=&Multiply_Float(
     A=&Add_Float(A=200, B=&Multiply_Float(A=150, B=$Particles.FakeDistance))
     B=&MaskFloatBySpawnGroup(Masked Float=1, Passthrough Float=-1, Spawn Group Mask=$Particles.PartitionParticles001.Partition)
    )
   ) {enabled=false}
   &AddVelocity(Velocity=$System.DeathVelocity, Velocity Speed Scale=0.8) {module=AddVelocity001}
   /SetVariables(Particles.VisibilityTag=$Particles.PartitionParticles.Partition) {isSetParameters=true, module=SetVariables_#6}
  ]
  particleUpdateScript [
   &ParticleState()
   &LineAttractionForce(Attraction Strength=-20, Line End:Vector3f=<0 -600 75>, Line Start:Vector3f=<0 600 75>, Use Attraction Falloff=false)
   &RegionMask(
    Invert Region=true
    Region Falloff=0.1
    Region Origin:Vector3f=&AddVectorToPosition(Position=$Engine.Emitter.SimulationPosition, Vector=<0 840 -100>)
    Region Radius=730
    UseOffset=false
    UseRotation=false
   ) {module=RegionMask002}
   &RegionMask(
    Invert Region=true
    Region Dimensions:Vector3f=<100 1200 1000>
    Region Falloff=0.1
    Region Origin:Vector3f=&AddVectorToPosition(Position=$Engine.Emitter.SimulationPosition, Vector=<0 -400 -100>)
    Region Type:ENiagaraDebugDrawShape=::NewEnumerator1
    Rotation:Vector3f=<0 5 0>
    Rotation Angle Type:ENiagara_AngleInput=::NewEnumerator0
    UseOffset=false
    UseRotation=true
   ) {module=RegionMask003}
   /SetVariables(Particles.CombinedMask:NiagaraFloat=&Multiply_Float(A=$Particles.RegionMask003.Float, B=$Particles.RegionMask002.Float)) {isSetParameters=true}
   /SetVariables(Particles.FakeHorizonDistance:NiagaraFloat=&RemapRange(
     Input Value To Remap=&MakeFloatFromVector(
      Vector=&ConvertPositionToVector(Input Position=$Particles.Position, Conversion Operation:ENiagara_LWCConvertPosToVec=::NewEnumerator0)
      Channel=::NewEnumerator2
     )
     Input Min=100
     Input Max=350
     Output Min=0
     Output Max=1
    )) {isSetParameters=true}
   &Drag(
    Drag=&RemapRange(Input Value To Remap=$Particles.FakeHorizonDistance, Input Min=0, Input Max=1, Output Min=0.25, Output Max=0)
    Use Linear Drag=true
    Use Rotational Drag=true
   )
   &ScaleMeshSize(Scale Factor=&Multiply_VectorByFloat(
     Vector=&Add_Vector(A=<1 1 1>, B=&Multiply_VectorByFloat(Vector=<1 1 1>, Float=$Particles.FakeHorizonDistance))
     Float=&Multiply_Float(
      A=$Particles.CombinedMask
      B=&FloatFromCurve(FloatCurve=DataInterface() {propertyValues {Curve {
          defaultValue=1
          keys [{interpMode=RCIM_Cubic, tangentMode=RCTM_User, time=0.7, value=1}, {arriveTangent=-5.7587, tangentMode=RCTM_User, time=1, value=0}]
         }}})
     )
    )) {module=ScaleMeshSize001}
   &SolveForcesAndVelocity(Clamp Velocity=false, Limit Acceleration=false)
   &KillParticlesInVolume(Box Size:Vector3f=<100 1800 1000>, Invert Volume=true, Kill Shape=::NewEnumerator1)
   &ScaleColor(
    Scale RGB=&MakeVectorFromLinearColorRGB(Color=&Multiply_LinearColorByFloat(
      Linear Color=<rgba 0.2464 0.2759 0.3 1>
      Float=&RemapRange(Input Value To Remap=$Particles.FakeHorizonDistance, Input Min=0, Input Max=1, Output Min=0.9, Output Max=0.15)
     ))
    ScaleA=true
    ScaleRGB=true
   )
  ]
  rendererClasses [&NiagaraSpriteRendererProperties, &NiagaraMeshRendererProperties]
  renderers [
   {rendererClass=&NiagaraSpriteRendererProperties}
   {rendererClass=&NiagaraMeshRendererProperties, rendererIndex=1}
  ]
 }
 XV_Projectile {
  emitterUpdateScript [
   &EmitterState(Life Cycle Mode=::NewEnumerator1, Loop Duration=1, UseLoopDelay=false)
   &SpawnRate(SpawnRate=4, Use Spawn Probability=false) {enabled=false}
   &SpawnParticlesFromOtherEmitter(
    Attribute Reader=DataInterface() {propertyValues {EmitterBinding {bindingMode=Other, emitterName=XV_Body}}}
    Spawn Rate=1
    Spawn Rate Per Particle Cap=1
    Use Maximum Age=false
    Use Minimum Age=false
   )
  ]
  particleSpawnScript [
   &InitializeParticle(
    Color=<rgba 2 0.0368 0.0001 1>
    Lifetime=&RandomRangeFloat(Minimum=1, Maximum=2)
    Mesh Scale Mode=::NewEnumerator3
    Mesh Uniform Scale=0.04
    UsePositionOffset=false
    Write Material Random=true
   ) {module=InitializeParticle001}
   &SampleParticlesFromOtherEmitter(
    Age Sampling:ENiagara_DirectReadSamplingMode=::NewEnumerator1
    Mesh Scale Sampling:ENiagara_DirectReadSamplingMode=::NewEnumerator1
    Velocity Sampling=::NewEnumerator1
   )
   &AddVelocity(
    Rotation Coordinate Space=::World
    Velocity=&NormalizeVector(
     Vector To Normalize=$Output.SampleParticlesFromOtherEmitter.SampledVelocity
     Safe Normalize=true
     Safe Fallback Threshold=0
     Safe Fallback Vector=<0 0 0.001>
    )
    Velocity Speed Scale=&Multiply_Float(A=700, B=1)
   )
   /SetVariables(Particles.Color=&MakeLinearColorFromVectorAndFloat(
     "Vector (RGB)"=&MakeVectorFromLinearColorRGB(Color=<rgba 12 0.2512 0 1>)
     "Float (Alpha)"=&FloatFromCurve(
      FloatCurve=DataInterface() {propertyValues {Curve {
         defaultValue=1
         keys [
          {arriveTangent=9.0725, interpMode=RCIM_Constant, time=0, value=0}
          <at 3 0.9847 RCIM_Constant>
          {arriveTangent=-0.0405, interpMode=RCIM_Constant, leaveTangent=-0.0405, time=8, value=0}
         ]
        }}}
      CurveIndex=$Output.SampleParticlesFromOtherEmitter.SampledAge
     )
    )) {enabled=false, isSetParameters=true}
   /SetVariables(Particles.Scale=&Add_Vector(
     A=&Multiply_VectorByFloat(
      Vector=&VectorFromFloat(Value=&MakeFloatFromVector(Vector=$Output.SampleParticlesFromOtherEmitter.SampledScale, Channel=::NewEnumerator0))
      Float=2
     )
     B=$Particles.Scale
    )) {isSetParameters=true}
  ]
  particleUpdateScript [
   &ParticleState()
   &SolveForcesAndVelocity(Clamp Velocity=false, Limit Acceleration=false)
   &RegionMask(
    Region Origin:Vector3f=&AddVectorToPosition(Position=$Engine.Emitter.SimulationPosition, Vector=<-660 0 -25>)
    Region Radius=500
    UseOffset=false
    UseRotation=false
   ) {module=RegionMask002}
   &RegionMask(
    Region Origin:Vector3f=&AddVectorToPosition(Position=$Engine.Emitter.SimulationPosition, Vector=<700 0 -400>)
    Region Radius=600
    UseOffset=false
    UseRotation=false
   ) {module=RegionMask003}
   &RegionMask(
    Region Origin:Vector3f=&AddVectorToPosition(Position=$Engine.Emitter.SimulationPosition, Vector=<250 0 -150>)
    Region Radius=300
    UseOffset=false
    UseRotation=false
   ) {module=RegionMask004}
   /SetVariables(Particles.CombinedMask:NiagaraFloat=&MakeCustomFloatFromBool(
     Bool=&Boolean_OrOperation(
      Bool1=&Boolean_OrOperation(Bool1=$Particles.RegionMask002.IsInsideMask, Bool2=$Particles.RegionMask003.IsInsideMask)
      Bool2=$Particles.RegionMask004.IsInsideMask
     )
     True Float=0
     False Float=1
    )) {isSetParameters=true}
   &ScaleMeshSize(Scale Factor=&Multiply_VectorByFloat(
     Vector=&MakeVector(
      X=&FloatFromCurve(
       FloatCurve=DataInterface() {propertyValues {Curve {
          defaultValue=1
          keys [
           {arriveTangent=4.1785, interpMode=RCIM_Cubic, leaveTangent=4.1785, tangentMode=RCTM_User, time=0, value=0}
           {arriveTangent=1.1179, interpMode=RCIM_Cubic, leaveTangent=1.1179, tangentMode=RCTM_User, time=0.2042, value=0.6228}
           {arriveTangent=0.0224, interpMode=RCIM_Cubic, leaveTangent=0.0224, tangentMode=RCTM_User, time=1, value=1}
          ]
         }}}
       Scale Curve=&RandomRangeFloat(Minimum=4, Maximum=6)
      )
      Y=2
      Z=2
     )
     Float=$Particles.CombinedMask
    ))
   /SetVariables(Particles.FakeHorizonDistance:NiagaraFloat=&RemapRange(
     Input Value To Remap=&MakeFloatFromVector(
      Vector=&ConvertPositionToVector(Input Position=$Particles.Position, Conversion Operation:ENiagara_LWCConvertPosToVec=::NewEnumerator0)
      Channel=::NewEnumerator2
     )
     Input Min=100
     Input Max=350
     Output Min=0
     Output Max=0.5
    )) {isSetParameters=true}
   &ScaleColor(Scale Alpha=&OneMinusFloat(Float=$Particles.FakeHorizonDistance), ScaleA=true, ScaleRGB=false)
   &KillParticlesInVolume(Box Size:Vector3f=<100 1800 1000>, Invert Volume=true, Kill Shape=::NewEnumerator1)
  ]
  rendererClasses [&NiagaraMeshRendererProperties]
  renderers [{rendererClass=&NiagaraMeshRendererProperties, rendererIndex=0}]
 }
}
summary {
 systemName=NS_S3500_Frontend
 emitters [
  {bEnabled=false, emitterName=Foreground_Fog, rendererClasses [&NiagaraSpriteRendererProperties]}
  {bEnabled=false, emitterName=SnowflakesAir, simTarget=GPUComputeSim, rendererClasses [&NiagaraSpriteRendererProperties]}
  {bEnabled=false, emitterName=SnowflakesAir_LOW, rendererClasses [&NiagaraSpriteRendererProperties]}
  {emitterName=TV_Body, rendererClasses [&NiagaraSpriteRendererProperties, &NiagaraMeshRendererProperties]}
  {emitterName=XV_Body, rendererClasses [&NiagaraSpriteRendererProperties, &NiagaraMeshRendererProperties]}
  {emitterName=XV_Projectile, rendererClasses [&NiagaraMeshRendererProperties]}
  {emitterName=TV_Projectile, rendererClasses [&NiagaraMeshRendererProperties]}
  {emitterName=Foreground_Fog001, rendererClasses [&NiagaraSpriteRendererProperties]}
  {emitterName=Foreground_Wisps, rendererClasses [&NiagaraSpriteRendererProperties]}
  {emitterName=Foreground_FlatSmoke, rendererClasses [&NiagaraSpriteRendererProperties]}
 ]
 userVariables []
}
